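# Behavior tree for Lonelies AI bots in vehicles, split into a driver branch and a branch for other seats.
# Drivers pick by alert level: drive aggressively toward attack destinations when threatened, follow last-known-position, or patrol while waiting for passengers, leaving the vehicle when conditions call for it.
# Other seats handle threatened, LKP, alerted and unaware states with weapon actions, observing, exiting or switching to the driver seat.
# Evaluator services write the blackboard keys that gate looping branches.
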BehaviorTree /Lonelies/BehaviorTree/BT_Vehicle_Lonelies
blackboard=&/BB_Lonelies
root {
 name=Selector
 class=BTComposite_Selector
 children [
  Sequence {
   class=BTComposite_Sequence
   children [
    FortAthenaBTTask_ModulateVehicleSpeed_1 {
     class=FortAthenaBTTask_ModulateVehicleSpeed
     children []
     decorators []
     properties {bIgnoreRestartSelf=false, bResetControlPIds=true, nodeName="", drivingSpeed {defaultValue=-1, key=None}}
     services []
    }
    Selector {
     class=BTComposite_Selector
     children [
      Selector {
       class=BTComposite_Selector
       children [
        Sequence {
         class=BTComposite_Sequence
         children [
          FortAthenaBTTask_SetAggressiveDriving_1 {
           class=FortAthenaBTTask_SetAggressiveDriving
           children []
           decorators []
           properties {bAggressiveDriving=true, bIgnoreRestartSelf=false, nodeName=""}
           services []
          }
          Loop Threaten {
           class=BTComposite_Selector
           children [
            LeaveVehicle {
             class=FortAthenaBTTask_LeaveVehicle
             children []
             decorators [Should leave vehicle {
               class=BTDecorator_Blackboard
               properties {
                arithmeticOperation=GreaterOrEqual
                intValue=4
                nodeName=Should leave vehicle
                stringValue=EExecutionStatus::ExecutionAllowed
                blackboardKey {selectedKeyName=AIEvaluator_LeaveSeatStatus}
               }
              }]
             properties {bIgnoreRestartSelf=false, bWaitVehicleStop=false, nodeName=LeaveVehicle}
             services []
            }
            Sequence {
             class=BTComposite_Sequence
             children [
              FortAthenaBTTask_ModulateVehicleSpeed_3 {
               class=FortAthenaBTTask_ModulateVehicleSpeed
               children []
               decorators []
               properties {bIgnoreRestartSelf=false, bResetControlPIds=true, nodeName="", drivingSpeed {defaultValue=-1, key=None}}
               services []
              }
              Simple Parallel {
               class=BTComposite_SimpleParallel
               children [
                Range Move To {
                 class=FortAthenaBTTask_BotMoveTo
                 decorators []
                 properties {
                  movementResultKeyName=AIEvaluator_RangeAttack_MovementState
                  nodeName=Range Move To
                  acceptableRadius {defaultValue=50}
                  allowRandomWobble {defaultValue=false}
                  bAllowStrafe {defaultValue=false}
                  blackboardKey {selectedKeyName=AIEvaluator_Attack_Destination}
                  observedBlackboardValueTolerance {defaultValue=200}
                 }
                }
                LeaveVehicle {
                 class=FortAthenaBTTask_LeaveVehicle
                 children []
                 decorators [Distance Between {
                   class=FortBTDecorator_DistanceBetween
                   properties {
                    flowAbortMode=None
                    nodeName=Distance Between
                    operator=LessOrEqual
                    blackboardKeyB {selectedKeyName=AIEvaluator_Global_LastKnownPosition}
                    specifiedDistance {defaultValue=1000}
                   }
                  }]
                 properties {bIgnoreRestartSelf=false, bWaitVehicleStop=true, nodeName=LeaveVehicle}
                 services []
                }
               ]
               properties {
                children [
                 {childTask=&self:FortAthenaBTTask_BotMoveTo_3}
                 {
                  childTask=&self:FortAthenaBTTask_LeaveVehicle_18
                  decorators [&self:FortBTDecorator_DistanceBetween_0]
                 }
                ]
                services []
               }
               services []
              }
              LeaveVehicle {
               class=FortAthenaBTTask_LeaveVehicle
               children []
               decorators [Distance Between {
                 class=FortBTDecorator_DistanceBetween
                 properties {
                  flowAbortMode=None
                  nodeName=Distance Between
                  operator=LessOrEqual
                  blackboardKeyB {selectedKeyName=AIEvaluator_Global_LastKnownPosition}
                  specifiedDistance {defaultValue=2000}
                 }
                }]
               properties {bIgnoreRestartSelf=false, bWaitVehicleStop=false, nodeName=LeaveVehicle}
               services []
              }
             ]
             decorators [
              Blackboard Based Condition {
               class=BTDecorator_Blackboard
               properties {
                arithmeticOperation=GreaterOrEqual
                flowAbortMode=None
                intValue=3
                nodeName=Blackboard Based Condition
                stringValue=EExecutionStatus::ExecutionPending
                blackboardKey {selectedKeyName=AIEvaluator_RangeAttack_ExecutionStatus}
               }
              }
              Blackboard Based Condition {
               class=BTDecorator_Blackboard
               properties {
                arithmeticOperation=GreaterOrEqual
                flowAbortMode=None
                intValue=3
                nodeName=Blackboard Based Condition
                stringValue=EExecutionStatus::ExecutionPending
                blackboardKey {selectedKeyName=AIEvaluator_ThrowableAttack_ExecutionStatus}
               }
              }
              Should move toward Target {
               class=BTDecorator_Blackboard
               properties {
                arithmeticOperation=Equal
                flowAbortMode=LowerPriority
                intValue=0
                nodeName=Should move toward Target
                stringValue=""
                blackboardKey {selectedKeyName=AIEvaluator_Attack_MoveToDestination}
               }
              }
              Attack Destination Is Valid {
               class=BTDecorator_Blackboard
               properties {
                arithmeticOperation=Equal
                flowAbortMode=LowerPriority
                intValue=0
                nodeName=Attack Destination Is Valid
                stringValue=""
                blackboardKey {selectedKeyName=AIEvaluator_Attack_Destination}
               }
              }
             ]
             properties {children [
               {childTask=&self:FortAthenaBTTask_ModulateVehicleSpeed_3}
               {childComposite=&self:BTComposite_SimpleParallel_0, childTask=null}
               {
                childTask=&self:FortAthenaBTTask_LeaveVehicle_17
                decorators [&self:FortBTDecorator_DistanceBetween_1]
               }
              ]}
            }
            Sequence {
             class=BTComposite_Sequence
             children [
              Wait {class=BTTask_Wait, properties {waitTime {defaultValue=1}}}
              LeaveVehicle {
               class=FortAthenaBTTask_LeaveVehicle
               children []
               decorators []
               properties {bIgnoreRestartSelf=false, bWaitVehicleStop=false, nodeName=LeaveVehicle}
               services []
              }
             ]
             decorators []
             properties {children [
               {childTask=&self:BTTask_Wait_14}
               {childTask=&self:FortAthenaBTTask_LeaveVehicle_3}
              ]}
            }
           ]
           decorators [Loop {class=BTDecorator_Loop}]
           properties {
            nodeName=Loop Threaten
            children [
             {
              childComposite=null
              childTask=&self:FortAthenaBTTask_LeaveVehicle_19
              decorators [&self:BTDecorator_Blackboard_42]
             }
             {
              childComposite=&self:BTComposite_Sequence_8
              childTask=null
              decoratorOps [{number=3, operation=And}, {number=2, operation=Or}, {number=0}, {number=1}, {number=2}, {number=3}]
              decorators [
               &self:BTDecorator_Blackboard_23
               &self:BTDecorator_Blackboard_22
               &self:BTDecorator_Blackboard_25
               &self:BTDecorator_Blackboard_36
              ]
             }
             {childComposite=&self:BTComposite_Sequence_13, childTask=null, decorators []}
            ]
            services [
             &self:FortAthenaBTService_AIEvaluator_31
             &self:FortAthenaBTService_AIEvaluator_32
            ]
           }
           services [
            Focus on target {
             class=FortAthenaBTService_AIEvaluator
             properties {interval=0.5, nodeName=Focus on target, aIEvaluatorClass {defaultValue=&FortAthenaAIBotEvaluator_HandleFocusing}}
            }
            Evaluating...Leave Vehicle {
             class=FortAthenaBTService_AIEvaluator
             properties {
              interval=0.5
              nodeName=Evaluating...Leave Vehicle
              aIEvaluatorClass {defaultValue=/Lonelies/Evaluators/Evaluator_VehicleLeaveSeat_Lonelies.Evaluator_VehicleLeaveSeat_Lonelies_C}
             }
            }
           ]
          }
         ]
         decorators [IsThreatened {
           class=BTDecorator_Blackboard
           properties {
            arithmeticOperation=Equal
            intValue=3
            nodeName=IsThreatened
            stringValue=EAlertLevel::Threatened
            blackboardKey {selectedKeyName=AIEvaluator_Global_AlertLevel}
           }
          }]
         properties {children [
           {childTask=&self:FortAthenaBTTask_SetAggressiveDriving_1}
           {
            childComposite=&self:BTComposite_Selector_21
            childTask=null
            decorators [&self:BTDecorator_Loop_11]
           }
          ]}
        }
        Sequence {
         class=BTComposite_Sequence
         children [
          FortAthenaBTTask_ActivateVehicleBoost_1 {
           class=FortAthenaBTTask_ActivateVehicleBoost
           children []
           decorators []
           properties {bActivateBoost=false, bIgnoreMinimumDistanceLeft=false, bIgnoreRestartSelf=false, boostLength=-1, nodeName=""}
           services []
          }
          Selector {
           class=BTComposite_Selector
           children [
            Range Move To {
             class=FortAthenaBTTask_BotMoveTo
             decorators [
              Should move toward Target {
               class=BTDecorator_Blackboard
               properties {
                arithmeticOperation=Equal
                flowAbortMode=LowerPriority
                intValue=0
                nodeName=Should move toward Target
                stringValue=""
                blackboardKey {selectedKeyName=AIEvaluator_Attack_MoveToDestination}
               }
              }
              Attack Destination Is Valid {
               class=BTDecorator_Blackboard
               properties {
                arithmeticOperation=Equal
                flowAbortMode=LowerPriority
                intValue=0
                nodeName=Attack Destination Is Valid
                stringValue=""
                blackboardKey {selectedKeyName=AIEvaluator_Attack_Destination}
               }
              }
             ]
             properties {
              movementResultKeyName=AIEvaluator_RangeAttack_MovementState
              nodeName=Range Move To
              acceptableRadius {defaultValue=50}
              allowRandomWobble {defaultValue=false}
              bAllowStrafe {defaultValue=false}
              blackboardKey {selectedKeyName=AIEvaluator_Attack_Destination}
              observedBlackboardValueTolerance {defaultValue=200}
             }
            }
            Sequence {
             class=BTComposite_Sequence
             children [
              Wait {class=BTTask_Wait, properties {randomDeviation {defaultValue=1}, waitTime {defaultValue=5}}}
              LeaveVehicle {
               class=FortAthenaBTTask_LeaveVehicle
               children []
               decorators []
               properties {bIgnoreRestartSelf=false, bWaitVehicleStop=false, nodeName=LeaveVehicle}
               services []
              }
             ]
             decorators []
             properties {children [
               {childTask=&self:BTTask_Wait_11}
               {childTask=&self:FortAthenaBTTask_LeaveVehicle_2}
              ]}
            }
           ]
           decorators [Loop {class=BTDecorator_Loop}]
           properties {
            children [
             {
              childComposite=null
              childTask=&self:FortAthenaBTTask_BotMoveTo_1
              decorators [&self:BTDecorator_Blackboard_16, &self:BTDecorator_Blackboard_20]
             }
             {childComposite=&self:BTComposite_Sequence_6, childTask=null, decorators []}
            ]
            services [&self:FortAthenaBTService_AIEvaluator_38]
           }
           services [Focus on target {
             class=FortAthenaBTService_AIEvaluator
             properties {interval=0.5, nodeName=Focus on target, aIEvaluatorClass {defaultValue=&FortAthenaAIBotEvaluator_HandleFocusing}}
            }]
          }
         ]
         decorators [IsLKP {
           class=BTDecorator_Blackboard
           properties {
            arithmeticOperation=Equal
            intValue=2
            nodeName=IsLKP
            stringValue=EAlertLevel::LKP
            blackboardKey {selectedKeyName=AIEvaluator_Global_AlertLevel}
           }
          }]
         properties {children [
           {childTask=&self:FortAthenaBTTask_ActivateVehicleBoost_1}
           {
            childComposite=&self:BTComposite_Selector_23
            childTask=null
            decorators [&self:BTDecorator_Loop_12]
           }
          ]}
        }
       ]
       decorators []
       properties {
        children [
         {
          childComposite=&self:BTComposite_Sequence_4
          childTask=null
          decorators [&self:BTDecorator_Blackboard_14]
         }
         {
          childComposite=&self:BTComposite_Sequence_20
          childTask=null
          decorators [&self:BTDecorator_Blackboard_43]
         }
        ]
        services [
         &self:FortAthenaBTService_AIEvaluator_13
         &self:FortAthenaBTService_AIEvaluator_40
        ]
       }
       services [
        Evaluating...Select Weapon {
         class=FortAthenaBTService_AIEvaluator
         properties {
          interval=0.2
          nodeName=Evaluating...Select Weapon
          randomDeviation=0.05
          aIEvaluatorClass {defaultValue=&FortAthenaAIBotEvaluator_WeaponSelection}
         }
        }
        Evaluating...Range Attack {
         class=FortAthenaBTService_AIEvaluator
         properties {
          bCallTickOnSearchStart=true
          interval=0.2
          nodeName=Evaluating...Range Attack
          randomDeviation=0.05
          aIEvaluatorClass {defaultValue=/Game/Athena/AI/Vehicle/Evaluators/VehicleRangeAttackEvaluator.VehicleRangeAttackEvaluator_C}
         }
        }
       ]
      }
      Sequence {
       class=BTComposite_Sequence
       children [
        FortAthenaBTTask_SetAggressiveDriving_0 {
         class=FortAthenaBTTask_SetAggressiveDriving
         children []
         decorators []
         properties {bAggressiveDriving=false, bIgnoreRestartSelf=false, nodeName=""}
         services []
        }
        Selector {
         class=BTComposite_Selector
         children [
          Wait {
           class=BTTask_Wait
           decorators [Need to Wait Passengers? {
             class=BTDecorator_Blackboard
             properties {
              arithmeticOperation=GreaterOrEqual
              flowAbortMode=Self
              intValue=4
              nodeName=Need to Wait Passengers?
              stringValue=EExecutionStatus::ExecutionAllowed
              blackboardKey {selectedKeyName=AIEvaluator_WaitForPassengerStatus}
             }
            }]
           properties {waitTime {defaultValue=3600}}
          }
          Selector {
           class=BTComposite_Selector
           children [
            LeaveVehicle {
             class=FortAthenaBTTask_LeaveVehicle
             children []
             decorators [Should Move? {
               class=BTDecorator_Blackboard
               properties {
                arithmeticOperation=GreaterOrEqual
                basicOperation=NotSet
                flowAbortMode=LowerPriority
                intValue=3
                nodeName=Should Move?
                stringValue=""
                blackboardKey {selectedKeyName=AIEvaluator_Patrolling_ShouldMove}
               }
              }]
             properties {bIgnoreRestartSelf=false, bWaitVehicleStop=false, nodeName=LeaveVehicle}
             services []
            }
            Simple Parallel {
             class=BTComposite_SimpleParallel
             children [
              MoveTo PatrolPoint {
               class=FortAthenaBTTask_BotMoveTo
               decorators []
               properties {
                movementResultKeyName=AIEvaluator_Patrolling_MovementState
                nodeName=MoveTo PatrolPoint
                acceptableRadius {defaultValue=10}
                allowRandomWobble {defaultValue=false}
                bAllowStrafe {defaultValue=true}
                bTrackMovingGoal {defaultValue=false}
                blackboardKey {selectedKeyName=AIEvaluator_Patrolling_Destination}
                observedBlackboardValueTolerance {defaultValue=50}
               }
              }
              Sequence {
               class=BTComposite_Sequence
               children [
                Wait {class=BTTask_Wait, properties {waitTime {defaultValue=4}}}
                LeaveVehicle {
                 class=FortAthenaBTTask_LeaveVehicle
                 children []
                 decorators []
                 properties {bIgnoreRestartSelf=false, bWaitVehicleStop=false, nodeName=LeaveVehicle}
                 services []
                }
               ]
               decorators [Should execute smart object behavior? {
                 class=BTDecorator_Blackboard
                 properties {
                  arithmeticOperation=GreaterOrEqual
                  flowAbortMode=None
                  intValue=4
                  nodeName=Should execute smart object behavior?
                  stringValue=EExecutionStatus::ExecutionAllowed
                  blackboardKey {selectedKeyName=AIEvaluator_SmartObject_ExecutionStatus}
                 }
                }]
               properties {
                children [
                 {childTask=&self:BTTask_Wait_2}
                 {childTask=&self:FortAthenaBTTask_LeaveVehicle_6}
                ]
                services [&self:FortAthenaBTService_PauseVehicle_3]
               }
               services [Executing...PauseVehicle {
                 class=FortAthenaBTService_PauseVehicle
                 properties {bCallTickOnSearchStart=false, bRestartTimerOnEachActivation=false, interval=0.5, nodeName=Executing...PauseVehicle, randomDeviation=0.1}
                }]
              }
             ]
             properties {
              children [
               {childTask=&self:FortAthenaBTTask_BotMoveTo_8}
               {
                childComposite=&self:BTComposite_Sequence_7
                childTask=null
                decorators [&self:BTDecorator_Blackboard_5]
               }
              ]
              services []
             }
             services []
            }
            Sequence {
             class=BTComposite_Sequence
             children [
              MoveTo PatrolPoint {
               class=FortAthenaBTTask_BotMoveTo
               decorators []
               properties {
                movementResultKeyName=AIEvaluator_Patrolling_MovementState
                nodeName=MoveTo PatrolPoint
                acceptableRadius {defaultValue=10}
                allowRandomWobble {defaultValue=false}
                bAllowStrafe {defaultValue=true}
                bTrackMovingGoal {defaultValue=false}
                blackboardKey {selectedKeyName=AIEvaluator_Patrolling_Destination}
                observedBlackboardValueTolerance {defaultValue=50}
               }
              }
              LeaveVehicle {
               class=FortAthenaBTTask_LeaveVehicle
               children []
               decorators [Can Run LocalGameplayBehavior? {
                 class=BTDecorator_Blackboard
                 properties {
                  arithmeticOperation=GreaterOrEqual
                  flowAbortMode=Self
                  intValue=3
                  nodeName=Can Run LocalGameplayBehavior?
                  stringValue=EExecutionStatus::ExecutionPending
                  blackboardKey {selectedKeyName=AIEvaluator_DynamicBlueprint_ExecutionStatus}
                 }
                }]
               properties {bIgnoreRestartSelf=false, bWaitVehicleStop=false, nodeName=LeaveVehicle}
               services []
              }
             ]
             decorators []
             properties {children [
               {childTask=&self:FortAthenaBTTask_BotMoveTo_0}
               {
                childTask=&self:FortAthenaBTTask_LeaveVehicle_0
                decorators [&self:BTDecorator_Blackboard_10]
               }
              ]}
            }
           ]
           decorators [Loop {class=BTDecorator_Loop}]
           properties {
            children [
             {
              childComposite=null
              childTask=&self:FortAthenaBTTask_LeaveVehicle_15
              decorators [&self:BTDecorator_Blackboard_40]
             }
             {childComposite=&self:BTComposite_SimpleParallel_1, childTask=null, decorators []}
             {childComposite=&self:BTComposite_Sequence_1, childTask=null, decorators []}
            ]
            services [
             &self:FortAthenaBTService_AIEvaluator_39
             &self:FortAthenaBTService_Patrolling_0
             &self:FortAthenaBTService_AIEvaluator_12
            ]
           }
           services [
            Evaluating... Patrolling {
             class=FortAthenaBTService_AIEvaluator
             properties {
              interval=0.25
              nodeName=Evaluating... Patrolling
              randomDeviation=0
              aIEvaluatorClass {defaultValue=/Script/FortniteAIServer.FortAthenaNpcEvaluator_Patrolling}
             }
            }
            Executing...Patrolling {
             class=FortAthenaBTService_Patrolling
             properties {
              acceptableRadius=5
              bAllowPartialPath=true
              bCallTickOnSearchStart=false
              bProjectGoalLocation=true
              bRestartTimerOnEachActivation=false
              filterClass=null
              interval=0.5
              nodeName=Executing...Patrolling
              patrollingAppendDestinationKeyName=AIEvaluator_Patrolling_AppendDestination
              randomDeviation=0.1
             }
            }
            Evaluating...Smart Object {
             class=FortAthenaBTService_AIEvaluator
             properties {
              interval=0.5
              nodeName=Evaluating...Smart Object
              aIEvaluatorClass {defaultValue=/Script/FortniteAIServer.FortAthenaAIBotEvaluator_SmartObjects}
             }
            }
           ]
          }
         ]
         decorators []
         properties {
          children [
           {
            childComposite=null
            childTask=&self:BTTask_Wait_0
            decorators [&self:BTDecorator_Blackboard_0]
           }
           {
            childComposite=&self:BTComposite_Selector_24
            childTask=null
            decorators [&self:BTDecorator_Loop_13]
           }
          ]
          services []
         }
         services []
        }
       ]
       decorators []
       properties {
        children [
         {childTask=&self:FortAthenaBTTask_SetAggressiveDriving_0}
         {childComposite=&self:BTComposite_Selector_1, childTask=null}
        ]
        services [
         &self:FortAthenaBTService_AIEvaluator_34
         &self:FortAthenaBTService_AIEvaluator_35
         &self:FortAthenaBTService_PropagatePatrolProgressToPassengers_0
         &self:FortAthenaBTService_WaitForPassengers_0
         &self:FortAthenaBTService_AIEvaluator_36
        ]
       }
       services [
        Evaluating... Holster Weapon {
         class=FortAthenaBTService_AIEvaluator
         properties {
          interval=1
          nodeName=Evaluating... Holster Weapon
          randomDeviation=0.2
          aIEvaluatorClass {defaultValue=/Script/FortniteAIServer.FortAthenaAIBotEvaluator_HolsterWeapon}
         }
        }
        Evaluating...Wait For Passengers {
         class=FortAthenaBTService_AIEvaluator
         properties {
          interval=0.5
          nodeName=Evaluating...Wait For Passengers
          aIEvaluatorClass {defaultValue=/Script/FortniteAIServer.FortAthenaAIBotEvaluator_WaitForPassengers}
         }
        }
        Executing...PropagatePatrolProgressToPassengers {
         class=FortAthenaBTService_PropagatePatrolProgressToPassengers
         properties {
          bCallTickOnSearchStart=false
          bRestartTimerOnEachActivation=false
          interval=0.5
          nodeName=Executing...PropagatePatrolProgressToPassengers
          patrollingAppendDestinationKeyName=AIEvaluator_Patrolling_AppendDestination
          randomDeviation=0.1
         }
        }
        Executing...WaitForPassengers {
         class=FortAthenaBTService_WaitForPassengers
         properties {
          bCallTickOnSearchStart=false
          bRestartTimerOnEachActivation=false
          interval=0.5
          nodeName=Executing...WaitForPassengers
          randomDeviation=0.1
          waitForPassengersStatusKeyName=AIEvaluator_WaitForPassengerStatus
         }
        }
        Focus on target {
         class=FortAthenaBTService_AIEvaluator
         properties {interval=0.5, nodeName=Focus on target, aIEvaluatorClass {defaultValue=&FortAthenaAIBotEvaluator_HandleFocusing}}
        }
       ]
      }
     ]
     decorators [Loop {class=BTDecorator_Loop}]
     properties {
      children [
       {childComposite=&self:BTComposite_Selector_25, childTask=null, decorators []}
       {childComposite=&self:BTComposite_Sequence_2, childTask=null, decorators []}
      ]
      services []
     }
     services []
    }
   ]
   decorators [Is Driving Vehicle {
     class=BTDecorator_Blackboard
     properties {
      arithmeticOperation=Equal
      intValue=0
      nodeName=Is Driving Vehicle
      notifyObserver=ValueChange
      stringValue=""
      blackboardKey {selectedKeyName=AIEvaluator_IsDrivingVehicle}
     }
    }]
   properties {children [
     {childTask=&self:FortAthenaBTTask_ModulateVehicleSpeed_1}
     {
      childComposite=&self:BTComposite_Selector_11
      childTask=null
      decorators [&self:BTDecorator_Loop_9]
     }
    ]}
  }
  Selector {
   class=BTComposite_Selector
   children [
    LeaveVehicle {
     class=FortAthenaBTTask_LeaveVehicle
     children []
     decorators [Should leave vehicle {
       class=BTDecorator_Blackboard
       properties {
        arithmeticOperation=GreaterOrEqual
        intValue=4
        nodeName=Should leave vehicle
        stringValue=EExecutionStatus::ExecutionAllowed
        blackboardKey {selectedKeyName=AIEvaluator_LeaveSeatStatus}
       }
      }]
     properties {bIgnoreRestartSelf=false, bWaitVehicleStop=false, nodeName=LeaveVehicle}
     services []
    }
    Selector {
     class=BTComposite_Selector
     children [
      Selector {
       class=BTComposite_Selector
       children [
        Sequence {
         class=BTComposite_Sequence
         children [
          Wait {class=BTTask_Wait, properties {randomDeviation {defaultValue=0.5}, waitTime {defaultValue=2}}}
          LeaveVehicle {
           class=FortAthenaBTTask_LeaveVehicle
           children []
           decorators []
           properties {bIgnoreRestartSelf=false, bWaitVehicleStop=false, nodeName=LeaveVehicle}
           services []
          }
         ]
         decorators [
          Can't fire weapon {
           class=BTDecorator_Blackboard
           properties {
            arithmeticOperation=LessOrEqual
            intValue=1
            nodeName=Can't fire weapon
            notifyObserver=ValueChange
            stringValue=EExecutionStatus::ExecutionDenied
            blackboardKey {selectedKeyName=AIEvaluator_WeaponFire_ExecutionStatus}
           }
          }
          Not Reloading {
           class=BTDecorator_Blackboard
           properties {
            arithmeticOperation=LessOrEqual
            intValue=1
            nodeName=Not Reloading
            stringValue=EExecutionStatus::ExecutionDenied
            blackboardKey {selectedKeyName=AIEvaluator_WeaponReload_ExecutionStatus}
           }
          }
          Should Switch Seat {
           class=BTDecorator_Blackboard
           properties {
            arithmeticOperation=GreaterOrEqual
            intValue=4
            nodeName=Should Switch Seat
            stringValue=EExecutionStatus::ExecutionAllowed
            blackboardKey {selectedKeyName=AIEvaluator_SwitchSeatStatus}
           }
          }
         ]
         properties {children [
           {childTask=&self:BTTask_Wait_13}
           {childTask=&self:FortAthenaBTTask_LeaveVehicle_5}
          ]}
        }
        Wait {class=BTTask_Wait, decorators [Loop {class=BTDecorator_Loop}], properties {waitTime {defaultValue=10}}}
       ]
       decorators [IsThreatened {
         class=BTDecorator_Blackboard
         properties {
          arithmeticOperation=Equal
          intValue=3
          nodeName=IsThreatened
          stringValue=EAlertLevel::Threatened
          blackboardKey {selectedKeyName=AIEvaluator_Global_AlertLevel}
         }
        }]
       properties {
        children [
         {
          childComposite=&self:BTComposite_Sequence_9
          childTask=null
          decorators [
           &self:BTDecorator_Blackboard_37
           &self:BTDecorator_Blackboard_38
           &self:BTDecorator_Blackboard_39
          ]
         }
         {
          childComposite=null
          childTask=&self:BTTask_Wait_6
          decorators [&self:BTDecorator_Loop_3]
         }
        ]
        services [
         &self:FortAthenaBTService_AIEvaluator_6
         &self:FortAthenaBTService_ManageWeapon_2
         &self:FortAthenaBTService_AIEvaluator_7
        ]
       }
       services [
        Evaluating...Vehicle Switch Seat {
         class=FortAthenaBTService_AIEvaluator
         properties {
          interval=0.5
          nodeName=Evaluating...Vehicle Switch Seat
          aIEvaluatorClass {defaultValue=/AICore/Humanoids/BehaviorTrees/Evaluators/VehicleSwitchSeatEvaluator.VehicleSwitchSeatEvaluator_C}
         }
        }
        Executing...Weapons Actions {class=FortAthenaBTService_ManageWeapon, properties {interval=1, randomDeviation=0.3}}
        Focus on target {
         class=FortAthenaBTService_AIEvaluator
         properties {interval=0.5, nodeName=Focus on target, aIEvaluatorClass {defaultValue=&FortAthenaAIBotEvaluator_HandleFocusing}}
        }
       ]
      }
      LKP {
       class=BTComposite_Selector
       children [Sequence {
         class=BTComposite_Sequence
         children [
          Wait {class=BTTask_Wait, properties {randomDeviation {defaultValue=1}, waitTime {defaultValue=5}}}
          LeaveVehicle {
           class=FortAthenaBTTask_LeaveVehicle
           children []
           decorators [Should Switch Seat {
             class=BTDecorator_Blackboard
             properties {
              arithmeticOperation=GreaterOrEqual
              flowAbortMode=Self
              intValue=4
              nodeName=Should Switch Seat
              stringValue=EExecutionStatus::ExecutionAllowed
              blackboardKey {selectedKeyName=AIEvaluator_SwitchSeatStatus}
             }
            }]
           properties {bIgnoreRestartSelf=false, bWaitVehicleStop=false, nodeName=LeaveVehicle}
           services []
          }
         ]
         decorators []
         properties {children [
           {childTask=&self:BTTask_Wait_5}
           {
            childTask=&self:FortAthenaBTTask_LeaveVehicle_1
            decorators [&self:BTDecorator_Blackboard_11]
           }
          ]}
        }]
       decorators [IsLKP {
         class=BTDecorator_Blackboard
         properties {
          arithmeticOperation=Equal
          intValue=2
          nodeName=IsLKP
          stringValue=EAlertLevel::LKP
          blackboardKey {selectedKeyName=AIEvaluator_Global_AlertLevel}
         }
        }]
       properties {
        nodeName=LKP
        children [{childComposite=&self:BTComposite_Sequence_0, childTask=null, decorators []}]
        services [
         &self:FortAthenaBTService_ManageWeapon_0
         &self:FortAthenaBTService_AIEvaluator_3
        ]
       }
       services [
        Executing...Weapons Actions {class=FortAthenaBTService_ManageWeapon, properties {interval=0.2, randomDeviation=0.1}}
        Evaluating...Vehicle Switch Seat {
         class=FortAthenaBTService_AIEvaluator
         properties {
          interval=0.5
          nodeName=Evaluating...Vehicle Switch Seat
          aIEvaluatorClass {defaultValue=&VehicleSwitchSeatEvaluator_C}
         }
        }
       ]
      }
     ]
     decorators []
     properties {
      children [
       {
        childComposite=&self:BTComposite_Selector_7
        childTask=null
        decorators [&self:BTDecorator_Blackboard_7]
       }
       {
        childComposite=&self:BTComposite_Selector_6
        childTask=null
        decorators [&self:BTDecorator_Blackboard_3]
       }
      ]
      services [
       &self:FortAthenaBTService_AIEvaluator_14
       &self:FortAthenaBTService_AIEvaluator_41
      ]
     }
     services [
      Evaluating...Select Weapon {
       class=FortAthenaBTService_AIEvaluator
       properties {
        interval=0.2
        nodeName=Evaluating...Select Weapon
        randomDeviation=0.05
        aIEvaluatorClass {defaultValue=&FortAthenaAIBotEvaluator_WeaponSelection}
       }
      }
      Evaluating...Range Attack {
       class=FortAthenaBTService_AIEvaluator
       properties {
        bCallTickOnSearchStart=true
        interval=0.2
        nodeName=Evaluating...Range Attack
        randomDeviation=0.05
        aIEvaluatorClass {defaultValue=&VehicleRangeAttackEvaluator_C}
       }
      }
     ]
    }
    ALERTED {
     class=BTComposite_Selector
     children [Loop Alerted {
       class=BTComposite_Selector
       children [
        Wait {
         class=BTTask_Wait
         decorators [Should Observe {
           class=BTDecorator_Blackboard
           properties {
            arithmeticOperation=GreaterOrEqual
            intValue=3
            nodeName=Should Observe
            stringValue=EExecutionStatus::ExecutionPending
            blackboardKey {selectedKeyName=AIEvaluator_Observe_ExecutionStatus}
           }
          }]
         properties {randomDeviation {defaultValue=0.25}, waitTime {defaultValue=1}}
        }
        FallBack {class=BTTask_Wait, properties {nodeName=FallBack, waitTime {defaultValue=1}}}
       ]
       decorators [Loop {class=BTDecorator_Loop}]
       properties {
        nodeName=Loop Alerted
        children [
         {
          childComposite=null
          childTask=&self:BTTask_Wait_7
          decorators [&self:BTDecorator_Blackboard_9]
         }
         {childComposite=null, childTask=&self:BTTask_Wait_10, decorators []}
        ]
        services []
       }
       services []
      }]
     decorators [IsAlerted {
       class=BTDecorator_Blackboard
       properties {
        arithmeticOperation=Equal
        intValue=1
        nodeName=IsAlerted
        stringValue=EAlertLevel::Alerted
        blackboardKey {selectedKeyName=AIEvaluator_Global_AlertLevel}
       }
      }]
     properties {
      nodeName=ALERTED
      children [{
        childComposite=&self:BTComposite_Selector_9
        childTask=null
        decorators [&self:BTDecorator_Loop_8]
       }]
      services [&self:FortAthenaBTService_AIEvaluator_8]
     }
     services [Evaluating...Observe Threat {
       class=FortAthenaBTService_AIEvaluator
       properties {
        bCallTickOnSearchStart=true
        interval=0.5
        nodeName=Evaluating...Observe Threat
        aIEvaluatorClass {defaultValue=/Script/FortniteAIServer.FortAthenaAIBotEvaluator_Observe}
       }
      }]
    }
    UNAWARE {
     class=BTComposite_Selector
     children [
      Selector {
       class=BTComposite_Selector
       children [
        Sequence {
         class=BTComposite_Sequence
         children [
          Wait {class=BTTask_Wait, properties {waitTime {defaultValue=4}}}
          LeaveVehicle {
           class=FortAthenaBTTask_LeaveVehicle
           children []
           decorators []
           properties {bIgnoreRestartSelf=false, bWaitVehicleStop=false, nodeName=LeaveVehicle}
           services []
          }
         ]
         decorators [Should execute smart object behavior? {
           class=BTDecorator_Blackboard
           properties {
            arithmeticOperation=GreaterOrEqual
            intValue=4
            nodeName=Should execute smart object behavior?
            stringValue=EExecutionStatus::ExecutionAllowed
            blackboardKey {selectedKeyName=AIEvaluator_SmartObject_ExecutionStatus}
           }
          }]
         properties {children [
           {childTask=&self:BTTask_Wait_1}
           {childTask=&self:FortAthenaBTTask_LeaveVehicle_4}
          ]}
        }
        FortAthenaBTTask_VehicleSwitchSeat_2 {
         class=FortAthenaBTTask_VehicleSwitchSeat
         children []
         decorators []
         properties {bIgnoreRestartSelf=false, nodeName="", switchSeatType=ToDriver}
         services []
        }
       ]
       decorators [Should Switch Seat {
         class=BTDecorator_Blackboard
         properties {
          arithmeticOperation=GreaterOrEqual
          intValue=4
          nodeName=Should Switch Seat
          stringValue=EExecutionStatus::ExecutionAllowed
          blackboardKey {selectedKeyName=AIEvaluator_SwitchSeatStatus}
         }
        }]
       properties {
        children [
         {
          childComposite=&self:BTComposite_Sequence_3
          childTask=null
          decorators [&self:BTDecorator_Blackboard_12]
         }
         {childComposite=null, childTask=&self:FortAthenaBTTask_VehicleSwitchSeat_2, decorators []}
        ]
        services [&self:FortAthenaBTService_AIEvaluator_2]
       }
       services [Evaluating...Smart Object {
         class=FortAthenaBTService_AIEvaluator
         properties {
          interval=0.5
          nodeName=Evaluating...Smart Object
          aIEvaluatorClass {defaultValue=&FortAthenaAIBotEvaluator_SmartObjects}
         }
        }]
      }
      Wait {class=BTTask_Wait, properties {waitTime {defaultValue=5}}}
     ]
     decorators []
     properties {
      nodeName=UNAWARE
      children [
       {
        childComposite=&self:BTComposite_Selector_0
        childTask=null
        decorators [&self:BTDecorator_Blackboard_1]
       }
       {childComposite=null, childTask=&self:BTTask_Wait_12, decorators []}
      ]
      services [
       &self:FortAthenaBTService_AIEvaluator_1
       &self:FortAthenaBTService_AIEvaluator_4
       &self:FortAthenaBTService_AIEvaluator_10
       &self:FortAthenaBTService_ManageWeapon_1
       &self:FortAthenaBTService_AIEvaluator_11
      ]
     }
     services [
      Evaluating...Select Weapon {
       class=FortAthenaBTService_AIEvaluator
       properties {
        interval=0.2
        nodeName=Evaluating...Select Weapon
        randomDeviation=0.05
        aIEvaluatorClass {defaultValue=&FortAthenaAIBotEvaluator_WeaponSelection}
       }
      }
      Evaluating...Range Attack {
       class=FortAthenaBTService_AIEvaluator
       properties {
        interval=1
        nodeName=Evaluating...Range Attack
        aIEvaluatorClass {defaultValue=&FortAthenaAIBotEvaluator_RangeAttack}
       }
      }
      Evaluating... Holster Weapon {
       class=FortAthenaBTService_AIEvaluator
       properties {
        interval=1
        nodeName=Evaluating... Holster Weapon
        randomDeviation=0.2
        aIEvaluatorClass {defaultValue=&FortAthenaAIBotEvaluator_HolsterWeapon}
       }
      }
      Executing...Weapons Actions {class=FortAthenaBTService_ManageWeapon, properties {interval=0.2, randomDeviation=0.1}}
      Evaluating...Vehicle Switch Seat {
       class=FortAthenaBTService_AIEvaluator
       properties {
        interval=0.5
        nodeName=Evaluating...Vehicle Switch Seat
        aIEvaluatorClass {defaultValue=&VehicleSwitchSeatEvaluator_C}
       }
      }
     ]
    }
   ]
   decorators [
    Is Driving Vehicle {
     class=BTDecorator_Blackboard
     properties {
      arithmeticOperation=Equal
      basicOperation=NotSet
      intValue=0
      nodeName=Is Driving Vehicle
      notifyObserver=ValueChange
      stringValue=""
      blackboardKey {selectedKeyName=AIEvaluator_IsDrivingVehicle}
     }
    }
    Loop {class=BTDecorator_Loop, properties {numLoops {defaultValue=1}}}
   ]
   properties {
    children [
     {
      childComposite=null
      childTask=&self:FortAthenaBTTask_LeaveVehicle_16
      decorators [&self:BTDecorator_Blackboard_41]
     }
     {childComposite=&self:BTComposite_Selector_26, childTask=null, decorators []}
     {
      childComposite=&self:BTComposite_Selector_8
      childTask=null
      decorators [&self:BTDecorator_Blackboard_8]
     }
     {childComposite=&self:BTComposite_Selector_10, childTask=null, decorators []}
    ]
    services [&self:FortAthenaBTService_AIEvaluator_0]
   }
   services [Evaluating...Leave Vehicle {
     class=FortAthenaBTService_AIEvaluator
     properties {
      interval=0.5
      nodeName=Evaluating...Leave Vehicle
      aIEvaluatorClass {defaultValue=&Evaluator_VehicleLeaveSeat_Lonelies_C}
     }
    }]
  }
 ]
 decorators []
 properties {
  children [
   {
    childComposite=&self:BTComposite_Sequence_5
    childTask=null
    decorators [&self:BTDecorator_Blackboard_15]
   }
   {
    childComposite=&self:BTComposite_Selector_4
    childTask=null
    decorators [&self:BTDecorator_Blackboard_2, &self:BTDecorator_Loop_2]
   }
  ]
  services []
 }
 services []
}
